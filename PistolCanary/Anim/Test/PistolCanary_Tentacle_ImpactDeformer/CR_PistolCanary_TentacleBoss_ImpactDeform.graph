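# Test Control Rig for the PistolCanary tentacle boss that feeds timed hit impacts into the DG_PistolCanary_ImpactDeformer mesh deformer.
# Incoming world-space hits with existing bones become aimed transforms, remapped from current to initial bone pose.
# A MaxHits-sized ring buffer stores them, retiring expired or oldest entries.
# Active locations, directions and start times drive the Optimus deformer, alongside debug drawing; the interaction event injects a test hit.
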ControlRigBlueprint /PistolCanary/Anim/Test/PistolCanary_Tentacle_ImpactDeformer/CR_PistolCanary_TentacleBoss_ImpactDeform
graphs [
 RigVMFunctionLibrary {links [], nodes []}
 RigVMModel {nodes [
   n1 = /Script/ControlRig.RigUnit_InteractionExecution() {is=struct RigVMUnitNode}
   n2 = /Script/ControlRig.RigUnit_AimBoneMath(
    DebugSettings=(bEnabled=False,Scale=10,WorldOffset=(Rotation=(X=0,Y=0,Z=0,W=1),Translation=(X=0,Y=0,Z=0),Scale3D=(X=1,Y=1,Z=1)))
    InputTransform=(Rotation=(X=0,Y=0,Z=0,W=1),Translation=(X=0,Y=1400,Z=0),Scale3D=(X=1,Y=1,Z=1))
    Primary=(Weight=1,Axis=(X=0,Y=0,Z=1),Target=(X=0,Y=0,Z=0),Kind=Location,Space=(Type=Bone,Name="None"))
    Secondary=(Weight=1,Axis=(X=1,Y=0,Z=0),Target=(X=0,Y=0,Z=1),Kind=Direction,Space=(Type=Bone,Name="None"))
    Weight=1
   ) {is=struct RigVMUnitNode}
   n3 = /Script/RigVM.RigVMFunction_MathMatrixToVectors(Value=/Script/RigVM.RigVMFunction_MathMatrixFromTransformV2(Value=@n2.Result) {is=struct RigVMUnitNode, template="Cast::Execute(in Value,out Result)"}.Result) {is=struct RigVMUnitNode}
   RigVMVariableNode(Value=((X=0,Y=0,Z=0)), ExecuteContext=@n1.ExecutePin, Value.0=@n3.Origin) {variable=HitLocationsWS}
   Scale = /Script/RigVM.RigVMFunction_MathVectorScale(Factor=-100, Value=@n3.Z) {is=struct RigVMUnitNode, template="Scale::Execute(in Value,in Factor,out Result)"}
   Random = /Script/RigVM.RigVMFunction_RandomFloat(Duration=0, Maximum=100, Minimum=0, Seed=217) {is=struct RigVMUnitNode, template="Random::Execute(in Seed,in Minimum,in Maximum,in Duration,out Result)"}
   > RigVMVariableNode(
    Value=((X=0,Y=0,Z=0))
    Value.0=/Script/RigVM.RigVMFunction_MathVectorAdd(
     B=(X=0,Y=0,Z=-100)
     A=&RigVMFunction_MathVectorAdd(A=@n3.Origin, B=@Scale.Result) {template="Add::Execute(in A,in B,out Result)"}.Result
     B.Z=&RigVMFunction_MathDoubleAdd(A=-100, B=@Random.Result).Result
    ) {is=struct RigVMUnitNode, template="Add::Execute(in A,in B,out Result)"}.Result
   ) {variable=HitTraceStartsWS}
   > RigVMVariableNode(Value=("root")) {variable=HitBones}
   n4 = &RigUnit_PrepareForExecution()
   n5 = RigVMVariableNode() {variable=MaxHits}
   Set_Num = RigVMDispatchNode(ExecuteContext=@n4.ExecutePin, Num=@n5.Value, Array=RigVMVariableNode() {variable=HitsRingBuffer}.Value) {template="DISPATCH_RigVMDispatch_ArraySetNum(io Array,in Num)"}
   > Set_Num_1 = RigVMDispatchNode(Num=@n5.Value, Array=RigVMVariableNode() {variable=HitTimesRingBuffer}.Value) {template="DISPATCH_RigVMDispatch_ArraySetNum(io Array,in Num)"}
   RigUnit_BeginExecution = &RigUnit_BeginExecution()
   n6 = &RigVMFunction_Sequence(ExecuteContext=@RigUnit_BeginExecution.ExecutePin)
   RigVMVariableNode(ExecuteContext=@n6.A, Value=/Script/RigVM.RigVMFunction_Timeline(Speed=1) {is=struct RigVMUnitNode}.Time) {variable=CurrentTime}
   n7 = &RigVMFunction_Sequence(ExecuteContext=@n6.B)
   GreaterEqual = /Script/RigVM.RigVMFunction_MathIntGreaterEqual(B=0, A=RigVMVariableNode() {variable=NumActiveHits}.Value) {is=struct RigVMUnitNode, template="GreaterEqual::Execute(in A,in B,out Result)"}
   n8 = &RigVMFunction_ControlFlowBranch(Condition=@GreaterEqual.Result, ExecuteContext=@n7.A)
   RigVMVariableNode(ExecuteContext=@n8.True, Value=RigVMVariableNode() {variable=StartIndex}.Value) {variable=LastStartIndex}
   > RigVMVariableNode(Value=RigVMVariableNode() {variable=NumActiveHits}.Value) {variable=LastNumActiveHits}
   > RigVMFunction_ForLoopCount = /Script/RigVM.RigVMFunction_ForLoopCount(Count=RigVMVariableNode() {variable=LastNumActiveHits}.Value) {is=struct RigVMUnitNode}
   n9 = /Script/RigVM.RigVMFunction_MathIntAdd(A=@RigVMFunction_ForLoopCount.Index, B=RigVMVariableNode() {variable=LastStartIndex}.Value) {is=struct RigVMUnitNode, template="Add::Execute(in A,in B,out Result)"}
   n10 = RigVMVariableNode() {variable=MaxHits}
   Subtract = &RigVMFunction_MathDoubleSub(
    B=RigVMDispatchNode(
     Index=/Script/RigVM.RigVMFunction_MathIntMod(A=@n9.Result, B=@n10.Value) {is=struct RigVMUnitNode, template="Modulo::Execute(in A,in B,out Result)"}.Result
     Array=RigVMVariableNode() {variable=HitTimesRingBuffer}.Value
    ) {template="DISPATCH_RigVMDispatch_ArrayGetAtIndex(in Array,in Index,out Element)"}.Element
    A=RigVMVariableNode() {variable=CurrentTime}.Value
   )
   n11 = &RigVMFunction_ControlFlowBranch(
    Condition=/Script/RigVM.RigVMFunction_MathDoubleGreater(A=@Subtract.Result, B=RigVMVariableNode() {variable=MaxDuration}.Value) {is=struct RigVMUnitNode, template="Greater::Execute(in A,in B,out Result)"}.Result
    ExecuteContext=@RigVMFunction_ForLoopCount.ExecutePin
   )
   Modulo = &RigVMFunction_MathIntMod(
    A=&RigVMFunction_MathIntAdd(B=1, A=RigVMVariableNode() {variable=StartIndex}.Value) {template="Add::Execute(in A,in B,out Result)"}.Result
    B=RigVMVariableNode() {variable=MaxHits}.Value
   ) {template="Modulo::Execute(in A,in B,out Result)"}
   RigVMVariableNode(Value=@Modulo.Result, ExecuteContext=@n11.True) {variable=StartIndex}
   > RigVMVariableNode(Value=/Script/RigVM.RigVMFunction_MathIntSub(B=1, A=RigVMVariableNode() {variable=NumActiveHits}.Value) {is=struct RigVMUnitNode, template="Subtract::Execute(in A,in B,out Result)"}.Result) {variable=NumActiveHits}
   n12 = &RigVMFunction_Sequence(ExecuteContext=@n7.B)
   RigVMDispatchNode(ExecuteContext=@n12.A, Array=RigVMVariableNode() {variable=ValidHits_ComponentSpace}.Value) {template="DISPATCH_RigVMDispatch_ArrayReset(io Array)"}
   > n13 = RigVMDispatchNode(Array=RigVMVariableNode() {variable=ValidHitBones}.Value) {template="DISPATCH_RigVMDispatch_ArrayReset(io Array)"}
   n14 = RigVMVariableNode() {variable=HitLocationsWS}
   n15 = RigVMVariableNode() {variable=HitBones}
   n16 = &RigVMFunction_ControlFlowBranch(
    Condition=RigVMDispatchNode(
     A=RigVMDispatchNode(Array=@n14.Value) {template="DISPATCH_RigVMDispatch_ArrayGetNum(in Array,out Num)"}.Num
     B=RigVMDispatchNode(Array=@n15.Value) {template="DISPATCH_RigVMDispatch_ArrayGetNum(in Array,out Num)"}.Num
    ) {template="DISPATCH_RigVMDispatch_CoreEquals(in A,in B,out Result)"}.Result
    ExecuteContext=@n13.ExecuteContext
   )
   For_Each_2 = RigVMDispatchNode(ExecuteContext=@n16.True, Array=RigVMVariableNode() {variable=HitLocationsWS}.Value) {template="DISPATCH_RigVMDispatch_ArrayIterator(in Array,out Element,out Index,out Count,out Ratio)"}
   n17 = RigVMDispatchNode(Index=@For_Each_2.Index, Array=RigVMVariableNode() {variable=HitBones}.Value) {template="DISPATCH_RigVMDispatch_ArrayGetAtIndex(in Array,in Index,out Element)"}
   n18 = &RigVMFunction_ControlFlowBranch(
    ExecuteContext=@For_Each_2.ExecuteContext
    Condition=/Script/ControlRig.RigUnit_ItemExists(Item=(Type=Bone,Name="None"), Item.Name=@n17.Element) {is=struct RigVMUnitNode}.Exists
   )
   From World = /Script/ControlRig.RigUnit_ToRigSpace_Location(Value=@For_Each_2.Element) {is=struct RigVMUnitNode, template="From World::Execute(in Value,out Global)"}
   From World_1 = &RigUnit_ToRigSpace_Location(Value=RigVMDispatchNode(Index=@For_Each_2.Index, Array=RigVMVariableNode() {variable=HitTraceStartsWS}.Value) {template="DISPATCH_RigVMDispatch_ArrayGetAtIndex(in Array,in Index,out Element)"}.Element) {template="From World::Execute(in Value,out Global)"}
   n19 = &RigUnit_AimBoneMath(
    DebugSettings=(bEnabled=False,Scale=10,WorldOffset=(Rotation=(X=0,Y=0,Z=0,W=1),Translation=(X=0,Y=0,Z=0),Scale3D=(X=1,Y=1,Z=1)))
    InputTransform=(Rotation=(X=0,Y=0,Z=0,W=1),Translation=(X=0,Y=0,Z=0),Scale3D=(X=1,Y=1,Z=1))
    Primary=(Weight=1,Axis=(X=0,Y=0,Z=-1),Target=(X=1,Y=0,Z=0),Kind=Location,Space=(Type=Bone,Name="None"))
    Secondary=(Weight=1,Axis=(X=0,Y=1,Z=0),Target=(X=0,Y=0,Z=1),Kind=Direction,Space=(Type=Bone,Name="None"))
    Weight=1
    InputTransform.Translation=@"From World".Global
    Primary.Target=@"From World_1".Global
   )
   RigVMDispatchNode(Element=@n19.Result, ExecuteContext=@n18.True, Array=RigVMVariableNode() {variable=ValidHits_ComponentSpace}.Value) {template="DISPATCH_RigVMDispatch_ArrayAdd(io Array,in Element,out Index)"}
   > RigVMDispatchNode(Element=@n17.Element, Array=RigVMVariableNode() {variable=ValidHitBones}.Value) {template="DISPATCH_RigVMDispatch_ArrayAdd(io Array,in Element,out Index)"}
   n20 = &RigVMFunction_Sequence(ExecuteContext=@n12.B)
   RigVMDispatchNode(ExecuteContext=@n20.A, Array=RigVMVariableNode() {variable=HitLocationsWS}.Value) {template="DISPATCH_RigVMDispatch_ArrayReset(io Array)"}
   > RigVMDispatchNode(Array=RigVMVariableNode() {variable=HitTraceStartsWS}.Value) {template="DISPATCH_RigVMDispatch_ArrayReset(io Array)"}
   > RigVMDispatchNode(Array=RigVMVariableNode() {variable=HitBones}.Value) {template="DISPATCH_RigVMDispatch_ArrayReset(io Array)"}
   > For_Each = RigVMDispatchNode(Array=RigVMVariableNode() {variable=ValidHits_ComponentSpace}.Value) {template="DISPATCH_RigVMDispatch_ArrayIterator(in Array,out Element,out Index,out Count,out Ratio)"}
   Add = &RigVMFunction_MathIntAdd(A=RigVMVariableNode() {variable=StartIndex}.Value, B=RigVMVariableNode() {variable=NumActiveHits}.Value) {template="Add::Execute(in A,in B,out Result)"}
   n21 = &RigVMFunction_MathIntMod(A=@Add.Result, B=RigVMVariableNode() {variable=MaxHits}.Value) {template="Modulo::Execute(in A,in B,out Result)"}
   n22 = RigVMDispatchNode(Index=@For_Each.Index, Array=RigVMVariableNode() {variable=ValidHitBones}.Value) {template="DISPATCH_RigVMDispatch_ArrayGetAtIndex(in Array,in Index,out Element)"}
   Inverse = /Script/RigVM.RigVMFunction_MathTransformInverse(Value=&RigUnit_GetTransform(Item=(Type=Bone,Name="None"), bInitial=False, Item.Name=@n22.Element).Transform) {is=struct RigVMUnitNode, template="Inverse::Execute(in Value,out Result)"}
   Multiply = &RigVMFunction_MathTransformMul(A=@For_Each.Element, B=@Inverse.Result) {template="Multiply::Execute(in A,in B,out Result)"}
   > Set_At = RigVMDispatchNode(
    Index=@n21.Result
    Element=&RigVMFunction_MathTransformMul(B=&RigUnit_GetTransform(Item=(Type=Bone,Name="None"), bInitial=true, Item.Name=@n22.Element).Transform, A=@Multiply.Result) {template="Multiply::Execute(in A,in B,out Result)"}.Result
    Array=RigVMVariableNode() {variable=HitsRingBuffer}.Value
   ) {template="DISPATCH_RigVMDispatch_ArraySetAtIndex(io Array,in Index,in Element)"}
   > Set_At_1 = RigVMDispatchNode(Index=@n21.Result, Array=RigVMVariableNode() {variable=HitTimesRingBuffer}.Value, Element=RigVMVariableNode() {variable=CurrentTime}.Value) {template="DISPATCH_RigVMDispatch_ArraySetAtIndex(io Array,in Index,in Element)"}
   n23 = &RigVMFunction_MathIntAdd(B=1, A=RigVMVariableNode() {variable=NumActiveHits}.Value) {template="Add::Execute(in A,in B,out Result)"}
   > n24 = RigVMVariableNode(Value=@n23.Result) {variable=NumActiveHits}
   Greater = /Script/RigVM.RigVMFunction_MathIntGreater(A=@n23.Result, B=RigVMVariableNode() {variable=MaxHits}.Value) {is=struct RigVMUnitNode, template="Greater::Execute(in A,in B,out Result)"}
   n25 = &RigVMFunction_ControlFlowBranch(Condition=@Greater.Result, ExecuteContext=@n24.ExecuteContext)
   n26 = RigVMVariableNode() {variable=StartIndex}
   RigVMVariableNode(
    Value=&RigVMFunction_MathIntMod(
     A=&RigVMFunction_MathIntAdd(B=1, A=@n26.Value) {template="Add::Execute(in A,in B,out Result)"}.Result
     B=RigVMVariableNode() {variable=MaxHits}.Value
    ) {template="Modulo::Execute(in A,in B,out Result)"}.Result
    ExecuteContext=@n25.True
   ) {variable=StartIndex}
   > RigVMVariableNode(Value=&RigVMFunction_MathIntSub(B=1, A=RigVMVariableNode() {variable=NumActiveHits}.Value) {template="Subtract::Execute(in A,in B,out Result)"}.Result) {variable=NumActiveHits}
   RigVMFunction_Sequence = &RigVMFunction_Sequence(ExecuteContext=@n20.B)
   n27 = &RigVMFunction_ControlFlowBranch(
    Condition=&RigVMFunction_MathIntGreater(B=0, A=RigVMVariableNode() {variable=NumActiveHits}.Value) {template="Greater::Execute(in A,in B,out Result)"}.Result
    ExecuteContext=@RigVMFunction_Sequence.A
   )
   For_Each_1 = RigVMDispatchNode(ExecuteContext=@n27.True, Array=RigVMVariableNode() {variable=HitTimesRingBuffer}.Value) {template="DISPATCH_RigVMDispatch_ArrayIterator(in Array,out Element,out Index,out Count,out Ratio)"}
   n28 = RigVMVariableNode() {variable=MaxHits}
   n29 = &RigVMFunction_MathIntMod(
    A=&RigVMFunction_MathIntSub(
     A=&RigVMFunction_MathIntAdd(A=@For_Each_1.Index, B=@n28.Value) {template="Add::Execute(in A,in B,out Result)"}.Result
     B=RigVMVariableNode() {variable=StartIndex}.Value
    ) {template="Subtract::Execute(in A,in B,out Result)"}.Result
    B=RigVMVariableNode() {variable=MaxHits}.Value
   ) {template="Modulo::Execute(in A,in B,out Result)"}
   n30 = RigVMVariableNode() {variable=NumActiveHits}
   Less = /Script/RigVM.RigVMFunction_MathIntLess(A=@n29.Result, B=@n30.Value) {is=struct RigVMUnitNode, template="Less::Execute(in A,in B,out Result)"}
   n31 = &RigVMFunction_MathDoubleMul(B=10000, A=@For_Each_1.Ratio)
   RigVMFunction_MathIntToDouble_1 = /Script/RigVM.RigVMFunction_MathIntToFloat(Value=@n30.Value) {is=struct RigVMUnitNode, template="Cast::Execute(in Value,out Result)"}
   RigVMFunction_MathIntToFloat = &RigVMFunction_MathIntToFloat(Value=@n29.Result) {template="Cast::Execute(in Value,out Result)"}
   Divide = &RigVMFunction_MathDoubleDiv(B=@RigVMFunction_MathIntToDouble_1.Result, A=@RigVMFunction_MathIntToFloat.Result)
   > RigVMFunction_DebugLineNoSpace = /Script/RigVM.RigVMFunction_DebugLineNoSpace(
    A=(X=0,Y=0,Z=0)
    B=(X=0,Y=0,Z=10)
    Color=(R=1,G=0,B=0,A=1)
    DebugDrawSettings=(DepthPriority=SDPG_Foreground,Lifetime=-1)
    Thickness=0
    WorldOffset=(Rotation=(X=0,Y=0,Z=0,W=1),Translation=(X=0,Y=0,Z=0),Scale3D=(X=1,Y=1,Z=1))
    bEnabled=true
    B.Z=RigVMDispatchNode(False=0, True=1000, Condition=@Less.Result) {template="DISPATCH_RigVMDispatch_If(in Condition,in True,in False,out Result)"}.Result
    A.X=@n31.Result
    B.X=@n31.Result
    Color.R=&RigVMFunction_MathDoubleSub(A=1, B=@Divide.Result).Result
   ) {is=struct RigVMUnitNode}
   n32 = &RigVMFunction_Sequence(ExecuteContext=@RigVMFunction_Sequence.B)
   Set_Num_4 = RigVMDispatchNode(Array=(), Num=0, ExecuteContext=@n32.A) {template="DISPATCH_RigVMDispatch_ArraySetNum(io Array,in Num)"}
   > Set_Num_5 = RigVMDispatchNode(Array=(), Num=0) {template="DISPATCH_RigVMDispatch_ArraySetNum(io Array,in Num)"}
   n33 = &RigVMFunction_ControlFlowBranch(
    Condition=&RigVMFunction_MathIntGreaterEqual(B=0, A=RigVMVariableNode() {variable=NumActiveHits}.Value) {template="GreaterEqual::Execute(in A,in B,out Result)"}.Result
    ExecuteContext=@Set_Num_5.ExecuteContext
   )
   n34 = &RigVMFunction_ForLoopCount(ExecutePin=@n33.True, Count=RigVMVariableNode() {variable=NumActiveHits}.Value)
   n35 = RigVMVariableNode() {variable=StartIndex}
   n36 = RigVMDispatchNode(
    Index=&RigVMFunction_MathIntMod(
     A=&RigVMFunction_MathIntAdd(A=@n34.Index, B=@n35.Value) {template="Add::Execute(in A,in B,out Result)"}.Result
     B=RigVMVariableNode() {variable=MaxHits}.Value
    ) {template="Modulo::Execute(in A,in B,out Result)"}.Result
    Array=RigVMVariableNode() {variable=HitsRingBuffer}.Value
   ) {template="DISPATCH_RigVMDispatch_ArrayGetAtIndex(in Array,in Index,out Element)"}
   RigVMFunction_MathMatrixFromTransformV2 = &RigVMFunction_MathMatrixFromTransformV2(Value=@n36.Element) {template="Cast::Execute(in Value,out Result)"}
   RigVMFunction_MathMatrixToVectors = &RigVMFunction_MathMatrixToVectors(Value=@RigVMFunction_MathMatrixFromTransformV2.Result)
   > RigVMDispatchNode(Element=@RigVMFunction_MathMatrixToVectors.Origin, Array=@Set_Num_4.Array) {template="DISPATCH_RigVMDispatch_ArrayAdd(io Array,in Element,out Index)"}
   > RigVMDispatchNode(Element=@RigVMFunction_MathMatrixToVectors.Z, Array=@Set_Num_5.Array) {template="DISPATCH_RigVMDispatch_ArrayAdd(io Array,in Element,out Index)"}
   > RigVMFunction_DebugTransformMutableNoSpace = /Script/RigVM.RigVMFunction_DebugTransformMutableNoSpace(
    Color=(R=1,G=0,B=0,A=1)
    DebugDrawSettings=(DepthPriority=SDPG_Foreground,Lifetime=-1)
    Mode=Axes
    Scale=1000
    Thickness=0
    WorldOffset=(Rotation=(X=0,Y=0,Z=0,W=1),Translation=(X=0,Y=0,Z=0),Scale3D=(X=1,Y=1,Z=1))
    bEnabled=True
    Transform=@n36.Element
   ) {is=struct RigVMUnitNode}
   n37 = &RigVMFunction_Sequence(ExecuteContext=@n32.B)
   Set_Num_2 = RigVMDispatchNode(Array=(), Num=0, ExecuteContext=@n37.A) {template="DISPATCH_RigVMDispatch_ArraySetNum(io Array,in Num)"}
   n38 = &RigVMFunction_ControlFlowBranch(
    Condition=&RigVMFunction_MathIntGreaterEqual(B=0, A=RigVMVariableNode() {variable=NumActiveHits}.Value) {template="GreaterEqual::Execute(in A,in B,out Result)"}.Result
    ExecuteContext=@Set_Num_2.ExecuteContext
   )
   n39 = &RigVMFunction_ForLoopCount(ExecutePin=@n38.True, Count=RigVMVariableNode() {variable=NumActiveHits}.Value)
   n40 = &RigVMFunction_MathIntAdd(A=@n39.Index, B=RigVMVariableNode() {variable=StartIndex}.Value) {template="Add::Execute(in A,in B,out Result)"}
   n41 = RigVMVariableNode() {variable=MaxHits}
   > RigVMDispatchNode(
    Element=RigVMDispatchNode(
     Index=&RigVMFunction_MathIntMod(A=@n40.Result, B=@n41.Value) {template="Modulo::Execute(in A,in B,out Result)"}.Result
     Array=RigVMVariableNode() {variable=HitTimesRingBuffer}.Value
    ) {template="DISPATCH_RigVMDispatch_ArrayGetAtIndex(in Array,in Index,out Element)"}.Element
    Array=@Set_Num_2.Array
   ) {template="DISPATCH_RigVMDispatch_ArrayAdd(io Array,in Element,out Index)"}
   n42 = &RigVMFunction_Sequence(ExecuteContext=@n37.B)
   n43 = &RigVMFunction_ControlFlowBranch(Condition=true, ExecuteContext=@n42.A)
   /Script/OptimusCore.RigUnit_AddOptimusDeformer(
    DeformerGraphAsset=(DeformerGraph="&/DG_PistolCanary_ImpactDeformer",Name="DeformerGraphAsset")
    Force=(Value=0,Name="Force")
    GameTime=(Value=0,Name="GameTime")
    HitDir=(Value=,Name="HitDir")
    HitLocation=(Value=,Name="HitLocation")
    HitStartTimes=(Value=,Name="HitStartTimes")
    MaxDuration=(Value=1,Name="MaxDuration")
    NumHits=(Value=0,Name="NumHits")
    Radius=(Value=0,Name="Radius")
    ResponseSpeed=(Value=0,Name="ResponseSpeed")
    Settings=(ExecutionPhase=OverrideDefaultDeformer,ExecutionGroup=1,DeformChildComponents=True,ExcludeChildComponentsWithTag="None",Name="Settings")
    Stiffness=(Value=0,Name="Stiffness")
    ExecutePin=@n43.True
    HitStartTimes.Value=@Set_Num_2.Array
    HitLocation.Value=@Set_Num_4.Array
    HitDir.Value=@Set_Num_5.Array
    ResponseSpeed.Value=RigVMVariableNode() {variable=ResponseSpeed_Normalized}.Value
    NumHits.Value=RigVMVariableNode() {variable=NumActiveHits}.Value
    MaxDuration.Value=RigVMVariableNode() {variable=MaxDuration}.Value
    Radius.Value=RigVMVariableNode() {variable=Radius_Normalized}.Value
    Stiffness.Value=RigVMVariableNode() {variable=Stiffness_Normalized}.Value
    Force.Value=RigVMVariableNode() {variable=Force_Normalized}.Value
    GameTime.Value=RigVMVariableNode() {variable=CurrentTime}.Value
   ) {is=struct RigVMUnitNode}
   n44 = &RigVMFunction_RandomFloat(Duration=0, Maximum=6.29, Minimum=0, Seed=217) {template="Random::Execute(in Seed,in Minimum,in Maximum,in Duration,out Result)"}
   Sin = /Script/RigVM.RigVMFunction_MathDoubleSin(Value=@n44.Result) {is=struct RigVMUnitNode, template="Sin::Execute(in Value,out Result)"}
   &RigVMFunction_MathDoubleMul(B=1500, A=@Sin.Result)
   Cos = /Script/RigVM.RigVMFunction_MathDoubleCos(Value=@n44.Result) {is=struct RigVMUnitNode, template="Cos::Execute(in Value,out Result)"}
   &RigVMFunction_MathDoubleMul(B=1500, A=@Cos.Result)
   &RigVMFunction_RandomFloat(Duration=0, Maximum=15000, Minimum=0, Seed=223) {template="Random::Execute(in Seed,in Minimum,in Maximum,in Duration,out Result)"}
  ]}
]
hierarchy [
 root {loc=<0 0 0>, rot=<0 0 0 1>}
 tentacle_base {parent=root, loc=<0.2999 -120.5468 -1636.4241>, rot=<-0.5 0.5 -0.5 -0.5>}
 tentacle_01 {parent=tentacle_base, loc=<1269.9846 -0.7843 -19.493>, rot=<0 -0.0043 0.0001 -1>}
 tentacle_02 {parent=tentacle_01, loc=<1279.0715 0 0>, rot=<0 0.0268 -0.0003 -0.9996>}
 tentacle_03 {parent=tentacle_02, loc=<1275.3529 0 0>, rot=<0 0.0031 0.0001 -1>}
 tentacle_04 {parent=tentacle_03, loc=<1119.8098 0 0>, rot=<0 -0.0112 0.0005 -0.9999>}
 tentacle_05 {parent=tentacle_04, loc=<1114.6722 0 0>, rot=<0 -0.0006 -0.0006 -1>}
 tentacle_06 {parent=tentacle_05, loc=<1126.2786 0 0>, rot=<0 0.0116 0.0009 -0.9999>}
 tentacle_07 {parent=tentacle_06, loc=<1115.1567 0 0>, rot=<0 -0.0294 -0.0008 -0.9996>}
 tentacle_08 {parent=tentacle_07, loc=<1113.4962 0 0>, rot=<0 0.0056 0.0001 -1>}
 tentacle_09 {parent=tentacle_08, loc=<1119.848 0 0>, rot=<0 -0.0065 0.0007 -1>}
 tentacle_10 {parent=tentacle_09, loc=<1112.4393 0 0>, rot=<0 -0.0094 -0.0007 -1>}
 tentacle_11 {parent=tentacle_10, loc=<1019.1152 0 0>, rot=<0 -0.0039 0.0004 -1>}
 tentacle_12 {parent=tentacle_11, loc=<696.9656 0 0>, rot=<0 0.0047 -0.0001 -1>}
 tentacle_13 {parent=tentacle_12, loc=<658.0217 0 0>, rot=<0 -0.0077 -0.0002 -1>}
 tentacle_14 {parent=tentacle_13, loc=<599.0422 0 0>, rot=<0 -0.0032 -0.0001 -1>}
 tentacle_15 {parent=tentacle_14, loc=<657.8021 0 0>, rot=<0 -0.0128 0 -0.9999>}
 tentacle_16 {parent=tentacle_15, loc=<541.2894 0 0>, rot=<0 -0.0132 0.0001 -0.9999>}
 tentacle_17 {parent=tentacle_16, loc=<466.9173 0 0>, rot=<0 -0.0119 0.0001 -0.9999>}
 eyeE {parent=tentacle_10, loc=<-14.0474 77.9413 27.2458>, rot=<0.6805 0.175 0.6875 0.1836>}
 eyeE_mover {parent=eyeE, loc=<0 0 0>, rot=<0 0 0 1>}
 eyeE_orb {parent=eyeE_mover, loc=<279.1516 0 0>, rot=<0.002 0 0 -1>}
 eyeE_innerlid_mstr {parent=eyeE_orb, loc=<0 0 0>, rot=<0 0 0 1>}
 eyeE_innerlidB_mstr {parent=eyeE_innerlid_mstr, loc=<0 0 0>, rot=<0 0.2335 0.0081 -0.9723>}
 eyeE_innerlidB {parent=eyeE_innerlidB_mstr, loc=<499.6037 0 0>, rot=<0 0 0 1>}
 eyeE_innerlidA_mstr {parent=eyeE_innerlid_mstr, loc=<0 0 0>, rot=<0 0 0 1>}
 eyeE_innerlidA {parent=eyeE_innerlidA_mstr, loc=<512.2459 0.1408 2.3499>, rot=<-0.0003 0.002 -0.0112 -0.9999>}
 eyeE_ball {parent=eyeE_orb, loc=<0 0 0>, rot=<0 0 0 1>}
 eyeE_lid_upper_master {parent=eyeE_mover, loc=<0 0 0>, rot=<-0.0021 -0.0018 -0.1264 -0.992>}
 eyeE_lid_upperB {parent=eyeE_lid_upper_master, loc=<700.47 -4.2134 208.1176>, rot=<-0.1076 0.0474 -0.0618 -0.9911>}
 eyeE_lid_upperA {parent=eyeE_lid_upper_master, loc=<735.8749 -16.6512 -187.5531>, rot=<0.1474 0.0032 -0.0296 -0.9886>}
 eyeE_lid_lower_master {parent=eyeE_mover, loc=<0 0 0>, rot=<0.9898 -0.1425 -0.0034 0.0021>}
 eyeE_lid_lowerB {parent=eyeE_lid_lower_master, loc=<711.381 -0.4981 -212.483>, rot=<-0.0422 -0.9926 0.1127 0.0184>}
 eyeE_lid_lowerA {parent=eyeE_lid_lower_master, loc=<744.1301 -15.9596 184.128>, rot=<-0.0643 -0.9902 -0.1238 0>}
 eyeD {parent=tentacle_08, loc=<792.9301 338.1037 -91.8744>, rot=<0.588 0.4008 0.5697 0.4111>}
 eyeD_mover {parent=eyeD, loc=<0 0 0>, rot=<0 0 0 1>}
 eyeD_orb {parent=eyeD_mover, loc=<0 -49.4335 0.0036>, rot=<-0.0436 0 0 -0.999>}
 eyeD_innerlid_mstr {parent=eyeD_orb, loc=<0 0 0>, rot=<0.0013 -0.0012 0.0125 -0.9999>}
 eyeD_innerlidA_mstr {parent=eyeD_innerlid_mstr, loc=<0 0 0>, rot=<0 0 0 1>}
 eyeD_innerlidA {parent=eyeD_innerlidA_mstr, loc=<782.9152 -8.2737 0.6861>, rot=<-0.0014 0.0056 0.0379 -0.9993>}
 eyeD_innerlidB_mstr {parent=eyeD_innerlid_mstr, loc=<0 0 0>, rot=<0 0.2271 0 -0.9739>}
 eyeD_innerlidB_adj {parent=eyeD_innerlidB_mstr, loc=<149.9942 -1.3209 0.0029>, rot=<0 0 0 1>}
 eyeD_innerlidB {parent=eyeD_innerlidB_adj, loc=<619.7855 -7.1135 2.6421>, rot=<0 0 0 1>}
 eyeD_ball {parent=eyeD_orb, loc=<0 0 0>, rot=<0 0 0 1>}
 eyeD_lid_upper_master {parent=eyeD_mover, loc=<0 0 0>, rot=<-0.0035 -0.0105 -0.2028 -0.9792>}
 eyeD_lid_upperB {parent=eyeD_lid_upper_master, loc=<694.2782 -71.0626 290.9466>, rot=<-0.1025 0.0648 0.014 -0.9925>}
 eyeD_lid_upperA {parent=eyeD_lid_upper_master, loc=<744.4085 -71.6916 -229.3011>, rot=<0.148 0.0009 0.0483 -0.9878>}
 eyeD_lid_lower_master {parent=eyeD_mover, loc=<0 0 0>, rot=<0.9815 -0.1914 -0.0002 0.0054>}
 eyeD_lid_lowerB {parent=eyeD_lid_lower_master, loc=<690.46 75.8062 -201.756>, rot=<0.0816 -0.9907 0.102 0.0383>}
 eyeD_lid_lowerA {parent=eyeD_lid_lower_master, loc=<721.8588 76.9731 216.7439>, rot=<0.0601 -0.9899 -0.1247 -0.0321>}
 eyeC {parent=tentacle_06, loc=<283.0168 -272.6368 324.0001>, rot=<0.7084 -0.1764 0.6616 -0.1713>}
 eyeC_mover {parent=eyeC, loc=<0 0 0>, rot=<0 0 0 1>}
 eyeC_orb {parent=eyeC_mover, loc=<0 0 0>, rot=<-0.002 0 0 -1>}
 eyeC_innerlid_mstr {parent=eyeC_orb, loc=<0 0 0>, rot=<0 0 0 1>}
 eyeC_innerlidA_mstr {parent=eyeC_innerlid_mstr, loc=<0 0 0>, rot=<0 0 0 1>}
 eyeC_innerlidA {parent=eyeC_innerlidA_mstr, loc=<791.9602 0.1806 1.1381>, rot=<0 0 0 1>}
 eyeC_innerlidB_mstr {parent=eyeC_innerlid_mstr, loc=<0 0 0>, rot=<0.0059 0.2135 0.0162 -0.9768>}
 eyeC_innerlidB {parent=eyeC_innerlidB_mstr, loc=<772.1907 0 0>, rot=<-0.0087 -0.0406 -0.015 -0.999>}
 eyeC_ball {parent=eyeC_orb, loc=<0 0 0>, rot=<0 0 0 1>}
 eyeC_lid_upper_master {parent=eyeC_mover, loc=<0 0 0>, rot=<0.9799 -0.1988 0.0167 0.0068>}
 eyeC_lid_upperB {parent=eyeC_lid_upper_master, loc=<732.4357 -41.6861 -259.556>, rot=<-0.1012 -0.9893 0.1051 0.0034>}
 eyeC_lid_upperA {parent=eyeC_lid_upper_master, loc=<734.7452 -41.0341 195.0583>, rot=<-0.118 -0.984 -0.1304 -0.0283>}
 eyeC_lid_lower_master {parent=eyeC_mover, loc=<0 0 0>, rot=<-0.0066 0.0184 -0.182 -0.9831>}
 eyeC_lid_lowerB {parent=eyeC_lid_lower_master, loc=<706.4946 -56.9594 257.4023>, rot=<-0.1105 0.0249 0.0829 -0.9901>}
 eyeC_lid_lowerA {parent=eyeC_lid_lower_master, loc=<713.7773 -48.6013 -203.3995>, rot=<0.1463 -0.0114 0.1094 -0.9831>}
 eyeB {parent=tentacle_05, loc=<-43.2429 -653.2473 24.393>, rot=<0.5989 -0.3999 0.5675 -0.3991>}
 eyeB_mover {parent=eyeB, loc=<0 0 0>, rot=<0 0 0 1>}
 eyeB_orb {parent=eyeB_mover, loc=<0 73.6008 -0.3782>, rot=<0.0026 0.0001 0.0538 -0.9985>}
 eyeB_innerlid_mstr {parent=eyeB_orb, loc=<0 0 0>, rot=<0 0 0 1>}
 eyeB_innerlidB_mstr {parent=eyeB_innerlid_mstr, loc=<0 0 0>, rot=<-0.009 0.2171 -0.0314 -0.9756>}
 eyeB_innerlidB_adj {parent=eyeB_innerlidB_mstr, loc=<150 0 0>, rot=<0 0 0 1>}
 eyeB_innerlidB {parent=eyeB_innerlidB_adj, loc=<650.024 0 0>, rot=<0.0144 -0.0444 0.0294 -0.9985>}
 eyeB_innerlidA_mstr {parent=eyeB_innerlid_mstr, loc=<0 0 0>, rot=<0 0 0 1>}
 eyeB_innerlidA {parent=eyeB_innerlidA_mstr, loc=<795.264 12.1153 1.1381>, rot=<0 0 0 1>}
 eyeB_ball {parent=eyeB_orb, loc=<0 0 0>, rot=<0 0 0 1>}
 eyeB_lid_upper_master {parent=eyeB_mover, loc=<0 -37.2459 0.1914>, rot=<0.9816 -0.1909 0.0016 0.0054>}
 eyeB_lid_upperB {parent=eyeB_lid_upper_master, loc=<699.8835 -21.9373 -314.8382>, rot=<-0.1345 -0.9847 0.1086 0.0239>}
 eyeB_lid_upperA {parent=eyeB_lid_upper_master, loc=<688.5106 -15.0076 353.548>, rot=<-0.1551 -0.9798 -0.1254 -0.0163>}
 eyeB_lid_lower_master {parent=eyeB_mover, loc=<0 -37.2459 0.1914>, rot=<-0.0074 0.0008 -0.2692 -0.9631>}
 eyeB_lid_lowerB {parent=eyeB_lid_lower_master, loc=<719.0813 49.9047 323.6022>, rot=<-0.0973 0.0542 0.0397 -0.993>}
 eyeB_lid_lowerA {parent=eyeB_lid_lower_master, loc=<697.4066 52.7256 -325.3439>, rot=<0.1519 -0.0157 0.0708 -0.9857>}
 eyeA {parent=tentacle_03, loc=<-121.0337 656.824 180.5193>, rot=<0.5892 0.4265 0.5465 0.4151>}
 eyeA_mover {parent=eyeA, loc=<0 0 0>, rot=<0 0 0 1>}
 eyeA_orb {parent=eyeA_mover, loc=<0 0 0>, rot=<0.0015 0 0 -1>}
 eyeA_innerlid_mstr {parent=eyeA_orb, loc=<0 0 0>, rot=<0 0 0 1>}
 eyeA_innerlidB_mstr {parent=eyeA_innerlid_mstr, loc=<0 0 0>, rot=<0.0028 0.2037 0.0189 -0.9789>}
 eyeA_innerlidB_adj {parent=eyeA_innerlidB_mstr, loc=<150 0 0>, rot=<0 0 0 1>}
 eyeA_innerlidB {parent=eyeA_innerlidB_adj, loc=<658.3829 0 0>, rot=<0 0 0 1>}
 eyeA_innerlidA_mstr {parent=eyeA_innerlid_mstr, loc=<0 0 0>, rot=<0 0 0 1>}
 eyeA_innerlidA {parent=eyeA_innerlidA_mstr, loc=<791.9602 0.1806 1.1381>, rot=<0 0 0 1>}
 eyeA_ball {parent=eyeA_orb, loc=<0 0 0>, rot=<0 0 0 1>}
 eyeA_lid_upper_master {parent=eyeA_mover, loc=<0 0 0>, rot=<0.0004 -0.0238 -0.3083 -0.951>}
 eyeA_lid_upperB {parent=eyeA_lid_upper_master, loc=<692.5844 -110.5035 363.1694>, rot=<-0.0959 0.0892 0.1207 -0.984>}
 eyeA_lid_upperA {parent=eyeA_lid_upper_master, loc=<754.2465 -99.584 -322.8323>, rot=<0.1457 -0.0017 0.1577 -0.9767>}
 eyeA_lid_lower_master {parent=eyeA_mover, loc=<0 0 0>, rot=<0.9699 -0.2425 -0.0214 0.0004>}
 eyeA_lid_lowerB {parent=eyeA_lid_lower_master, loc=<699.9528 10.0732 -360.9964>, rot=<-0.1408 -0.9827 0.11 0.0478>}
 eyeA_lid_lowerA {parent=eyeA_lid_lower_master, loc=<761.6666 32.2089 268.7746>, rot=<-0.1669 -0.9783 -0.123 0.0054>}
 Eye_X_Offset {type=Curve, loc=<0 0 0>, rot=<0 0 0 1>}
 Eye_Y_Offset {type=Curve, loc=<0 0 0>, rot=<0 0 0 1>}
 TestTrigger {controlType=EULER_TRANSFORM, type=Control, loc=<0 0 0>, rot=<0 0 0 1>}
]
variables [
 CurrentTime {type=double, default=0}
 Force_Normalized {type=double, default=0}
 HitBones {type="TArray<FName>"}
 HitLocationsWS {type="TArray<FVector>"}
 HitTimesRingBuffer {type="TArray<double>"}
 HitTraceStartsWS {type="TArray<FVector>"}
 HitsRingBuffer {type="TArray<FTransform>"}
 LastNumActiveHits {type=int32, default=0}
 LastStartIndex {type=int32, default=0}
 MaxDuration {type=double, default=1}
 MaxHits {type=int32, default=30}
 NumActiveHits {type=int32, default=0}
 Radius_Normalized {type=double, default=0}
 ResponseSpeed_Normalized {type=double, default=0}
 StartIndex {type=int32, default=0}
 Stiffness_Normalized {type=double, default=0}
 ValidHitBones {type="TArray<FName>"}
 ValidHits_ComponentSpace {type="TArray<FTransform>"}
]
